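# Physics asset for the StrikinglyBright Fortnite backpack cosmetic, providing bodies and joints for its dangling keychain and straps.
# Two boxes form the pack body, with four small base capsules jointed to it with free swing and twist.
# From those bases hang a three-link keychain ending in a box, two two-link long straps and a two-link short strap.
# Tight swing and near-zero twist limits constrain each chain, keeping its motion subtle and mostly directional.
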PhysicsAsset /BRCosmetics/Accessories/FORT_Backpacks/Backpack_F_MED_StrikinglyBright/Meshes/F_MED_StrikinglyBright_Pack_Physics_fix
bodies {
 base_keychain [{shapeType=Capsule, length=0.505, radius=0.505, shapeName=base_keychain_capsule, rotation=<rot 0 -90 0>}]
 base_longstrap_A [{shapeType=Capsule, length=0.505, radius=0.505, shapeName=base_longstrap_A_capsule, rotation=<rot 0 -90 0>}]
 base_longstrap_B [{shapeType=Capsule, length=0.505, radius=0.505, shapeName=base_longstrap_B_capsule, rotation=<rot 0 -90 0>}]
 base_strap_B [{shapeType=Capsule, length=0.505, radius=0.505, shapeName=base_strap_B_capsule, rotation=<rot 0 -90 0>}]
 dyn_keychain_01 [{shapeType=Capsule, length=0.5, radius=0.5, shapeName=dyn_keychain_01_capsule, center=<0.5 0 0>, rotation=<rot 90 0 0>}]
 dyn_keychain_02 [{shapeType=Capsule, length=0.5, radius=0.5, shapeName=dyn_keychain_02_capsule, center=<0.5 0 0>, rotation=<rot 90 0 0>}]
 dyn_keychain_03 [{
   shapeType=Box
   extentX=0.9281
   extentY=4.1382
   extentZ=6.3567
   shapeName=dyn_keychain_03_box
   center=<2.6125 0 -0.5217>
   rotation=<rot 0 90 -50>
  }]
 dyn_longstrap_A_01 [{shapeType=Capsule, length=4, radius=1, shapeName=dyn_longstrap_A_01_capsule, center=<2 0 0>, rotation=<rot 90 0 0>}]
 dyn_longstrap_A_02 [{shapeType=Capsule, length=5.0443, radius=1.1423, shapeName=dyn_longstrap_A_02_capsule, center=<4.7813 0 -0.9594>, rotation=<rot 80 0 0>}]
 dyn_longstrap_B_01 [{shapeType=Capsule, length=4, radius=1, shapeName=dyn_longstrap_B_01_capsule, center=<2 0 0>, rotation=<rot 90 0 0>}]
 dyn_longstrap_B_02 [{shapeType=Capsule, length=4, radius=1, shapeName=dyn_longstrap_B_02_capsule, center=<3 0 0>, rotation=<rot 90 0 0>}]
 dyn_strap_B_01 [{shapeType=Capsule, length=1, radius=1, shapeName=dyn_strap_B_01_capsule, rotation=<rot 90 0 0>}]
 dyn_strap_B_02 [{shapeType=Capsule, length=6.9518, radius=1, shapeName=dyn_strap_B_02_capsule, center=<0.8981 0.3853 0.2856>, rotation=<rot 0 0 90>}]
 pack [
  {
   shapeType=Box
   extentX=13.2588
   extentY=14.6329
   extentZ=21.9445
   shapeName=pack_box
   center=<-3.7115 -8.9438 1.8205>
   rotation=<rot -0.101 90.0721 -39.7522>
  }
  {
   shapeType=Box
   extentX=2.9045
   extentY=26.3029
   extentZ=28.0024
   shapeName=pack_box
   center=<-1.6853 2.6373 0.3186>
   rotation=<rot 0.0002 77.6521 0>
  }
 ]
}
constraints [
 {bone1Name=dyn_strap_B_02, bone2Name=dyn_strap_B_01, swing1LimitDegrees=0, swing2LimitDegrees=15, twistLimitDegrees=0}
 {bone1Name=dyn_strap_B_01, bone2Name=base_strap_B, swing1LimitDegrees=0, swing2LimitDegrees=30, twistLimitDegrees=0}
 {bone1Name=dyn_longstrap_B_02, bone2Name=dyn_longstrap_B_01, swing1LimitDegrees=8, swing2LimitDegrees=2, twistLimitDegrees=0.5}
 {bone1Name=dyn_longstrap_B_01, bone2Name=base_longstrap_B, swing1LimitDegrees=15, swing2LimitDegrees=5, twistLimitDegrees=1}
 {bone1Name=dyn_longstrap_A_02, bone2Name=dyn_longstrap_A_01, swing1LimitDegrees=3, swing2LimitDegrees=1, twistLimitDegrees=0.5}
 {bone1Name=dyn_keychain_03, bone2Name=dyn_keychain_02, swing1LimitDegrees=30, swing2LimitDegrees=30, twistLimitDegrees=5}
 {bone1Name=dyn_keychain_02, bone2Name=dyn_keychain_01, swing1LimitDegrees=30, swing2LimitDegrees=5, twistLimitDegrees=0}
 {bone1Name=dyn_keychain_01, bone2Name=base_keychain, swing1LimitDegrees=5, swing2LimitDegrees=30, twistLimitDegrees=0}
 {
  bone1Name=base_longstrap_B
  bone2Name=pack
  swing1LimitDegrees=45
  swing1Motion=Free
  swing2LimitDegrees=45
  swing2Motion=Free
  twistLimitDegrees=45
  twistMotion=Free
 }
 {
  bone1Name=base_longstrap_A
  bone2Name=pack
  swing1LimitDegrees=45
  swing1Motion=Free
  swing2LimitDegrees=45
  swing2Motion=Free
  twistLimitDegrees=45
  twistMotion=Free
 }
 {
  bone1Name=base_keychain
  bone2Name=pack
  swing1LimitDegrees=45
  swing1Motion=Free
  swing2LimitDegrees=45
  swing2Motion=Free
  twistLimitDegrees=45
  twistMotion=Free
 }
 {
  bone1Name=base_strap_B
  bone2Name=pack
  swing1LimitDegrees=45
  swing1Motion=Free
  swing2LimitDegrees=45
  swing2Motion=Free
  twistLimitDegrees=45
  twistMotion=Free
 }
 {bone1Name=dyn_longstrap_A_01, bone2Name=base_longstrap_A, swing1LimitDegrees=10, swing2LimitDegrees=5, twistLimitDegrees=1}
]
